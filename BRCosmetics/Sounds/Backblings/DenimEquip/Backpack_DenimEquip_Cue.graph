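# Sound cue for the equip sound of the Denim backpack backbling cosmetic.
# A fixed short delay feeds a random node that picks one of three equally weighted wave variants.
# It uses the reactive backbling property sheet, the backbling sound class and close-range pet attenuation.
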SoundCue /BRCosmetics/Sounds/Backblings/DenimEquip/Backpack_DenimEquip_Cue
root=Delay_1
nodes [
 {id=Delay_1, children [Random_0], properties {delayMax=0.3, delayMin=0.3}}
 {id=Random_0, children [WavePlayer_0,WavePlayer_1,WavePlayer_2], properties {weights [1,1,1]}}
 {id=WavePlayer_0, properties {soundWaveAssetPtr=&/Backpack_DenimEquip_01}}
 {id=WavePlayer_1, properties {soundWaveAssetPtr=&/Backpack_DenimEquip_02}}
 {id=WavePlayer_2, properties {soundWaveAssetPtr=&/Backpack_DenimEquip_03}}
]
properties {
 attenuationSettings=&Fort_Pets_Close_Athena_Att
 audioPropertiesSheet=/BRCosmetics/Sounds/Settings/PropertySheets/Reactive/APS_Cosmetic_Reactive_Backbling
 maxDistance=1800.0001
 soundClassObject=/Game/Sounds/FortSoundClassesAndModes/Fort_Backblings
 volumeMultiplier=1
}
